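# Light function material for time-of-day cloud shadows, with an added triangular shadow that travels across the world, driven by MPC_SharpFrog.
# Three scrolling samples of one cloud texture blend by camera distance and fade out with world height.
# The moving shadow is a procedural equilateral-triangle distance field, interpolated between start and end positions with timed fades.
# The platform switch omits that shadow on Switch, Android and lower PC tiers, leaving only clouds.
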Material /Game/TimeOfDay/TODM_Sequence/Materials/Sky/Clouds/LF_cloudshadow_SharpFrog
graph {
 properties {materialDomain=MD_LightFunction, pixelDepthOffsetMode=PDOM_Legacy, usedShadingModels=Default Lit}
 nodes [
  n1 = CollectionParameter() {collection=&/MPC_SharpFrog, parameterName=Shadow Motion Time}
  Add(A=@n1, B=ScalarParameter() {parameterName=Shadow Motion Extra Time})
  MaterialFunctionCall(Float4=CollectionParameter() {collection=&MPC_SharpFrog, parameterName=Position}) {materialFunction=&BreakOutFloat4Components}
  n2 = TextureObject() {texture=&/T_cloudshadow2}
  n3 = MaterialFunctionCall!(
   PreviewInMatEditor
   Size=ScalarParameter() {defaultValue=220000, parameterName=scale 3}
   Speed=Multiply(A=ScalarParameter() {defaultValue=-0.15, parameterName=Speed3}, B=NamedRerouteUsage())
   Resolution=TextureProperty(Tex=@n2)
  ) {materialFunction=/Game/TimeOfDay/LightFunction/MF_CloudLayerCoord}
  n4 = TextureSample!(UVs=@n3.UVs, Tex=@n2, Level=@n3.UVs, Apply View MipBias) {
   mipValueMode=TMVM_MipLevel
   samplerSource=SSM_Wrap_WorldGroupSettings
   texture=&T_cloudshadow2
  }
  n5 = WorldPosition()
  n6 = Distance(A=CameraPositionWS(), B=@n5.XYZ)
  n7 = ScalarParameter() {defaultValue=5000, parameterName=NearRadiusSmallCloud}
  n8 = Subtract(A=@n6, B=@n7)
  n9 = Multiply(A=@n4.R, B=Clamp(None=Divide(A=@n8, B=@n7)))
  n10 = TextureObject() {texture=&T_cloudshadow2}
  n11 = MaterialFunctionCall(
   Size=ScalarParameter() {defaultValue=17500, parameterName=scale 2}
   Speed=Multiply(A=ScalarParameter() {defaultValue=-1, parameterName=Speed}, B=NamedRerouteUsage())
   Resolution=TextureProperty(Tex=@n10)
  ) {materialFunction=&MF_CloudLayerCoord}
  n12 = TextureSample(UVs=@n11.UVs, Tex=@n10, Level=@n11.UVs) {
   mipValueMode=TMVM_MipLevel
   samplerSource=SSM_Wrap_WorldGroupSettings
   texture=&T_cloudshadow2
  }
  n13 = LinearInterpolate(B=ScalarParameter() {defaultValue=0.7, parameterName=shadowdark}, Alpha=Saturate(None=Max(A=@n9, B=@n12.R))) {constA=1, constB=1.01}
  n14 = TextureObject() {texture=&T_cloudshadow2}
  n15 = MaterialFunctionCall(
   Size=ScalarParameter() {defaultValue=190000, parameterName=scale 1}
   Speed=Multiply(A=ScalarParameter() {defaultValue=-0.4, parameterName=Speed2}, B=NamedRerouteUsage())
   Resolution=TextureProperty(Tex=@n14)
  ) {materialFunction=&MF_CloudLayerCoord}
  n16 = TextureSample(UVs=@n15.UVs, Tex=@n14, Level=@n15.UVs) {
   mipValueMode=TMVM_MipLevel
   samplerSource=SSM_Wrap_WorldGroupSettings
   texture=&T_cloudshadow2
  }
  n17 = LinearInterpolate(B=ScalarParameter() {defaultValue=0.4, parameterName=shadowdark1}, Alpha=@n16.R) {constA=1, constB=1.01}
  n18 = ScalarParameter() {defaultValue=10000, parameterName=NearRadius}
  n19 = Divide(A=Subtract(A=@n6, B=@n18), B=@n18)
  n20 = Subtract(A=@n5.Z, B=ScalarParameter() {defaultValue=15000, parameterName=HeightFadeOffset})
  n21 = ScalarParameter() {defaultValue=10000, parameterName=HeightFadeBlend}
  n22 = LinearInterpolate(
   A=LinearInterpolate(A=Power(Base=@n13), B=Power(Base=@n17), Alpha=Clamp(None=@n19))
   B=Constant() {r=1}
   Alpha=Saturate(None=Divide(A=@n20, B=@n21))
  )
  n23 = CollectionParameter() {collection=&MPC_SharpFrog, parameterName=Shadow Edge Min}
  n24 = CollectionParameter() {collection=&MPC_SharpFrog, parameterName=Shadow Edge Max}
  n25 = NamedRerouteUsage()
  n26 = NamedRerouteUsage()
  n27 = Multiply(A=NamedRerouteUsage(), B=@n26)
  n28 = Subtract(A=@n25, B=@n27)
  n29 = NamedRerouteUsage()
  n30 = Subtract(A=Multiply(A=Multiply(A=@n29, B=@n26)) {constB=-1}, B=NamedRerouteUsage())
  n31 = LinearInterpolate(
   A=Multiply(A=AppendVector(A=@n28, B=@n30)) {constB=0.5}
   B=AppendVector(A=NamedRerouteUsage(), B=NamedRerouteUsage())
   Alpha=Step(Y=Add(A=@n27, B=@n25)) {constX=0}
  )
  n32 = ComponentMask(None=@n31) {r=true}
  n33 = Multiply(A=NamedRerouteUsage()) {constB=-2}
  n34 = ComponentMask(None=@n31) {g=true}
  n35 = AppendVector(A=Subtract(A=@n32, B=Clamp(None=@n32, Min=@n33) {maxDefault=0}), B=@n34)
  n36 = Multiply(A=Multiply(A=Length(None=@n35) {is=class MaterialExpressionLength}) {constB=-1}, B=Sign(None=@n34))
  n37 = OneMinus(None=Saturate(None=InverseLinearInterpolate(A=@n23, B=@n24, Value=@n36)))
  n38 = CollectionParameter() {collection=&MPC_SharpFrog, parameterName=Shadow Edge Power}
  n39 = WorldPosition()
  n40 = ScalarParameter() {defaultValue=20000, parameterName=Fake Shadow Height Falloff Offset}
  n41 = Saturate(None=Divide(A=Subtract(A=@n39.Z, B=@n40), B=ScalarParameter() {defaultValue=20000, parameterName=Fake Shadow Height Falloff Range}) {constB=40000})
  n42 = Saturate(None=Multiply(A=Power(Base=@n37, Exp=@n38), B=OneMinus(None=@n41)))
  n43 = LinearInterpolate(A=@n22, B=Multiply(A=@n22, B=OneMinus(None=@n42)), Alpha=NamedRerouteUsage()) {constB=0}
  MaterialFunctionCall!("Xb1,PS4"=@n43, "PC-High&&Epic,XSX,PS5"=@n43, "PC-Perf,Android"=@n22, Switch=@n22, "PC-Low(NotPerf)&&Medium"=@n22) {materialFunction=/Game/Effects/Niagara/MasterMaterials/Functions/Common/PlatformSwitch_3D}
  NamedRerouteDeclaration(None=ScalarParameter() {defaultValue=1.5, parameterName=GlobalSpeed}) {name=GlobalSpeed, nodeColor=<rgba 1 0 0.4452 1>}
  n44 = Subtract(
   A=Time()
   B=CollectionParameter() {collection=&MPC_SharpFrog, parameterName=Start Game Time}
  )
  NamedRerouteDeclaration(None=Saturate(None=Divide(A=@n44, B=@n1))) {name=Shadow Motion 0-1, nodeColor=<rgba 1 0 0.6797 1>}
  n45 = NamedRerouteUsage()
  n46 = NamedRerouteUsage()
  n47 = NamedRerouteUsage()
  NamedRerouteDeclaration(None=Add(A=ComponentMask(None=@n45) {g=true}, B=Divide(A=@n46, B=@n47))) {name=Height, nodeColor=<rgba 0.5994 0 1 1>}
  n48 = Divide(B=@n1)
  n49 = Multiply(A=@n48, B=ScalarParameter() {parameterName=Fake Shadow Fade In Offset In Seconds})
  n50 = Multiply(A=@n48, B=ScalarParameter() {defaultValue=0.25, parameterName=Fake Shadow Fade In Speed In Seconds})
  n51 = NamedRerouteUsage()
  n52 = Saturate(None=InverseLinearInterpolate(A=@n49, B=Add(A=@n50, B=@n49), Value=@n51) {constB=0.01})
  n53 = Multiply(A=@n48, B=ScalarParameter() {parameterName=Fake Shadow Fade Out Offset In Seconds})
  n54 = Multiply(A=@n48, B=ScalarParameter() {defaultValue=0.5, parameterName=Fake Shadow Fade Out Speed In Seconds})
  n55 = Saturate(None=InverseLinearInterpolate(A=@n53, B=Add(A=@n54, B=@n53), Value=OneMinus(None=@n51)) {constB=0.01})
  NamedRerouteDeclaration(None=Multiply(
    A=Multiply(A=@n52, B=@n55)
    B=CollectionParameter() {collection=&MPC_SharpFrog, parameterName=Shadow Enabled}
   )) {name=Shadow Visibility, nodeColor=<rgba 0 1 1 1>}
  n56 = CollectionParameter() {collection=&MPC_SharpFrog, parameterName=Shadow Start Position}
  n57 = CollectionParameter() {collection=&MPC_SharpFrog, parameterName=Shadow End Position}
  n58 = NamedRerouteUsage()
  n59 = MaterialFunctionCall(Float2=ComponentMask(None=LinearInterpolate(A=@n56, B=@n57, Alpha=@n58)) {g=true, r=true}) {materialFunction=&BreakOutFloat2Components}
  n60 = ScalarParameter() {defaultValue=168881.5938, parameterName=Fake Shadow Scale}
  n61 = Add(A=AppendVector(A=@n59.R, B=@n59.R), B=Multiply(A=@n60) {constB=0.5})
  n62 = WorldPosition()
  n63 = CollectionParameter() {collection=&MPC_SharpFrog, parameterName=Shadow Rotation}
  n64 = ScalarParameter() {defaultValue=0.25, parameterName=Fake Shadow Additional Roation}
  n65 = MaterialFunctionCall(UVs=Divide(A=Subtract(A=@n61, B=@n62.XY), B=@n60), "Rotation Angle (0-1)"=Multiply(A=Add(A=@n63, B=@n64) {constB=0.5}) {constB=-1}) {materialFunction=&CustomRotator}
  n66 = MaterialFunctionCall(Float2=@n65.Rotated Values) {materialFunction=&BreakOutFloat2Components}
  n67 = ConstantBiasScale(None=@n66.R) {bias=-0.5, scale=2}
  n68 = Divide(B=ScalarParameter() {defaultValue=0.8, parameterName=Fake Shadow Width})
  n69 = Multiply(A=Divide(B=ScalarParameter() {defaultValue=0.6, parameterName=Fake Shadow Height}), B=@n66.G)
  n70 = ScalarParameter() {defaultValue=-0.6, parameterName=Fake Shadow Offset}
  NamedRerouteDeclaration(None=AppendVector(A=Multiply(A=@n67, B=@n68), B=Add(A=@n69, B=@n70))) {name=UVs, nodeColor=<rgba 0.0447 0 1 1>}
  NamedRerouteDeclaration(None=CollectionParameter() {collection=&MPC_SharpFrog, parameterName=Shadow Edge Radius}) {name=Radius, nodeColor=<rgba 1 0.375 0 1>}
  NamedRerouteDeclaration(None=SquareRoot(None=Constant() {r=3})) {name=sqrt3, nodeColor=<rgba 1 0 0.3516 1>}
  n71 = ComponentMask(None=@n45) {r=true}
  NamedRerouteDeclaration(None=Subtract(A=Abs(None=@n71), B=NamedRerouteUsage())) {name=Width, nodeColor=<rgba 0.2592 1 0.6809 1>}
  OneMinus(None=Saturate(None=CollectionParameter() {collection=&MPC_SharpFrog, parameterName=Shadow Opacity}))
  ScalarParameter() {defaultValue=1, parameterName=Fake Shadow Opacity}
  ScalarParameter() {parameterName=Fake Shadow Position X}
  ScalarParameter() {parameterName=Fake Shadow Position Y}
  n72 = TextureObjectParameter() {parameterName=Fake Shadow Texture, texture=/Game/Athena/HUD/Art/Tutorial/FTUE_Arrow_Triangle}
  TextureSample(
   UVs=@n65.Rotated Values
   Tex=@n72
   Level=MaterialFunctionCall(Size=@n60, Resolution=TextureProperty(Tex=@n72)) {materialFunction=&MF_CloudLayerCoord}.MIP
  ) {mipValueMode=TMVM_MipLevel, samplerSource=SSM_Clamp_WorldGroupSettings}
 ]
}
